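# Looping sound cue for the Season 33 Turtle quest shard, playing as a quest-class sound effect.
# A single looping wave player feeds a modulator that adds slight random pitch and volume variation each time the cue starts.
# It uses the chest attenuation asset for 3D falloff, with boosted volume and a slightly lowered pitch.
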SoundCue /BattlePassS33_Quests/Audio/SFX/Quest_S33_Turtle/Quest_S33_Turtle_Shard_Loop_Cue
root=Modulator_0
nodes [
 {id=Modulator_0, children [WavePlayer_2], properties {pitchMax=1.02, pitchMin=0.98, volumeMax=1.02, volumeMin=0.98}}
 {
  id=WavePlayer_2
  properties {bLooping=true, soundWaveAssetPtr=&/Quest_S33_Turtle_Shard_Loop_01}
 }
]
properties {
 attenuationSettings=/Game/Sounds/Attenuation/Fort_Chests_Att
 maxDistance=1300
 pitchMultiplier=0.98
 soundClassObject=&Fort_Quests
 volumeMultiplier=2
}
