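# Sound cue for the shard shatter effect in the Season 33 Battle Pass Turtle quest.
# It picks one of two shatter waves at equal odds, then a modulator adds slight random pitch and volume variation.
# It plays in the quests sound class at reduced volume and slightly lowered pitch, borrowing the player pickaxe impact attenuation for distance falloff.
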SoundCue /BattlePassS33_Quests/Audio/SFX/Quest_S33_Turtle/Quest_S33_Turtle_Shard_Shatter_Cue
root=Modulator_0
nodes [
 {id=Modulator_0, children [Random_0], properties {pitchMax=1.02, pitchMin=0.98, volumeMax=1.02, volumeMin=0.98}}
 {id=Random_0, children [WavePlayer_0,WavePlayer_1], properties {weights [1,1]}}
 {id=WavePlayer_0, properties {soundWaveAssetPtr=&/Quest_S33_Turtle_Shard_Shatter_01}}
 {id=WavePlayer_1, properties {soundWaveAssetPtr=&/Quest_S33_Turtle_Shard_Shatter_02}}
]
properties {
 attenuationSettings=/Game/Sounds/Attenuation/Fort_Weapon_Pickaxe_Player_Impact_Att
 maxDistance=3512
 pitchMultiplier=0.98
 soundClassObject=&Fort_Quests
 volumeMultiplier=0.65
}
